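# PistolCanary_tac5200_aud is an audio-only Level Sequence for tac5200, apparently the sound layer of a tentacle boss sequence.
# A root-level TentacleAudio track holds one section with default volume and pitch, no sound asset shown and no attach actor set.
# A TentacleBoss possessable is bound but carries no tracks of its own, so the audio track is not nested under it.
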LevelSequence /PistolCanary/Sequence/tac/tac5200/PistolCanary_tac5200_aud
bindings [TentacleBoss {id=#1, possessed_class=PistolCanary_TentacleBoss_C, tracks []}]
display_rate {denominator=1, numerator=60}
tick_resolution {denominator=1, numerator=24000}
tracks [{
  type=MovieSceneAudioTrack
  display_name=TentacleAudio
  sections [{
    type=MovieSceneAudioSection
    end_frame=4036
    start_frame=3600
    channels [
     Volume {type=MovieSceneScriptingFloatChannel, default=1}
     Pitch {type=MovieSceneScriptingFloatChannel, default=1}
     AttachActor {type=MovieSceneScriptingActorReferenceChannel}
    ]
   }]
 }]
